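# Niagara swing trail for the CyborgWarrior pickaxe, built from three ribbon emitters that use the anim-trail modules on bone or socket index pairs.
# MainTrail and a darker DarkTrail each spawn two trails on the same index setup, tinted green by color curves.
# HandleTrail adds a near-black trail between another pair, drawn by two ribbon renderers.
# Short-lived particles fade by alpha curves, with time-of-day scalars dimming the green trails outside daytime.
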NiagaraSystem /BRCosmetics/Weapons/FORT_Melee/Pickaxe_CyborgWarrior/FX/NS_Pickaxe_CyborgWarrior_Trail
dependencies {
 usedDataInterfaces [&NiagaraDataInterfaceColorCurve, &NiagaraDataInterfaceCurve, &NiagaraDataInterfaceSocketReader]
 usedDynamicInputs []
 usedModules [
  &self:SystemSpawnScript.SystemScriptSource.SystemScriptGraph.NiagaraNodeAssignment_0.SetVariables_#1
  &SystemState
 ]
}
emitters {
 DarkTrail {
  emitterUpdateScript [
   &EmitterState()
   &FN_SpawnAnimTrail(Number of Anim Trails=2)
  ]
  particleSpawnScript [
   &FN_InitializeAnimTrailBoneIndecies(First Bone or Socket Index=1, Second Bone or Socket Index=3)
   &FN_InitializeAnimTrailBoneIndecies(AnimTrail Selection=1, First Bone or Socket Index=3) {module=FN_InitializeAnimTrailBoneIndecies001}
   &FN_InitializeAnimTrail(AnimTrailWidth=1.5, Lifetime=0.2)
  ]
  particleUpdateScript [
   &SolveForcesAndVelocity(Clamp Velocity=false, Limit Acceleration=false)
   &FN_UpdateAnimTrail()
   &ScaleColor(
    Scale Alpha=&FloatFromCurve(
     FloatCurve=DataInterface() {propertyValues {Curve {keys [
         {arriveTangent=-0.2226, interpMode=RCIM_Cubic, leaveTangent=-0.2226, tangentMode=RCTM_User, time=0.1, value=1}
         {arriveTangent=-0.0236, tangentMode=RCTM_User, time=1, value=0}
        ]}}}
     Scale Curve=&Multiply_Float(A=1.5, B=&FNTimeOfDayScalar(Night Scale=0.3, Morning Scale=0.4, Evening Scale=0.4))
    )
    Scale RGB=&Multiply_Vector(
     A=<1 1 1>
     B=&MakeVectorFromLinearColorRGB(Color=&ColorFromCurve(ColorCurve=DataInterface() {propertyValues {
         AlphaCurve {keys [<at 0 1 RCIM_Cubic>]}
         BlueCurve {keys [<at 0 0.0265>,<at 0.2812 0.0199>,<at 0.7034 0.0132>]}
         GreenCurve {keys [<at 0 0.2>,<at 0.2812 0.15>,<at 0.7034 0.1>]}
         RedCurve {keys [<at 0 0>,<at 0.2812 0>,<at 0.7034 0>]}
        }}))
    )
    ScaleA=true
    ScaleRGB=true
   )
   &SpriteFacingAndAlignment(Sprite Facing=$Particles.RibbonFacing, WriteAlignment=false, WriteFacing=true)
  ]
  rendererClasses [&NiagaraRibbonRendererProperties]
  renderers [{rendererClass=&NiagaraRibbonRendererProperties}]
 }
 HandleTrail {
  emitterUpdateScript [
   &EmitterState()
   &FN_SpawnAnimTrail(AnimTrailWidthMode=::FromSecond)
  ]
  particleSpawnScript [
   &FN_InitializeAnimTrailBoneIndecies(First Bone or Socket Index=4, Second Bone or Socket Index=5)
   &FN_InitializeAnimTrail(Color=<rgba 0.015 0.015 0.015 1>, Lifetime=0.2)
  ]
  particleUpdateScript [
   &SolveForcesAndVelocity(Clamp Velocity=false, Limit Acceleration=false)
   &FN_UpdateAnimTrail()
   &ScaleColor(
    Scale Alpha=&FloatFromCurve(
     FloatCurve=DataInterface() {propertyValues {Curve {keys [
         {arriveTangent=-0.2226, interpMode=RCIM_Cubic, leaveTangent=-0.2226, tangentMode=RCTM_User, time=0.1, value=1}
         {arriveTangent=-0.0236, tangentMode=RCTM_User, time=0.5, value=0}
        ]}}}
     Scale Curve=&Multiply_Float(A=1.5, B=&FNTimeOfDayScalar(Night Scale=1, Morning Scale=1, Evening Scale=1))
    )
    ScaleA=true
    ScaleRGB=true
   )
   &SpriteFacingAndAlignment(Sprite Facing=$Particles.RibbonFacing, WriteAlignment=false, WriteFacing=true)
  ]
  rendererClasses [&NiagaraRibbonRendererProperties]
  renderers [
   {rendererClass=&NiagaraRibbonRendererProperties}
   {rendererClass=&NiagaraRibbonRendererProperties, rendererIndex=1}
  ]
 }
 MainTrail {
  emitterUpdateScript [
   &EmitterState()
   &FN_SpawnAnimTrail(AnimTrailWidthMode=::FromSecond, Number of Anim Trails=2)
  ]
  particleSpawnScript [
   &FN_InitializeAnimTrailBoneIndecies(First Bone or Socket Index=1, Second Bone or Socket Index=3)
   &FN_InitializeAnimTrailBoneIndecies(AnimTrail Selection=1, First Bone or Socket Index=3) {module=FN_InitializeAnimTrailBoneIndecies001}
   &FN_InitializeAnimTrail(AnimTrailWidth=2, Color=<rgba 1 1 1 0.8>, Lifetime=0.2)
  ]
  particleUpdateScript [
   &SolveForcesAndVelocity(Clamp Velocity=false, Limit Acceleration=false)
   &FN_UpdateAnimTrail()
   &ScaleColor(
    Scale Alpha=&FloatFromCurve(
     FloatCurve=DataInterface() {propertyValues {Curve {keys [{arriveTangent=-2.1531, interpMode=RCIM_Cubic, leaveTangent=-2.1531, tangentMode=RCTM_User, time=0, value=1}, <at 1 0 RCIM_Cubic>]}}}
     Scale Curve=&FNTimeOfDayScalar(Night Scale=0.3, Morning Scale=0.3, Day Scale=0.5, Evening Scale=0.3)
    )
    Scale RGB=&Multiply_Vector(
     A=<1 1 1>
     B=&MakeVectorFromLinearColorRGB(Color=&ColorFromCurve(ColorCurve=DataInterface() {propertyValues {
         AlphaCurve {keys [<at 0 1 RCIM_Cubic>]}
         BlueCurve {keys [<at 0 0.0666>,<at 0.2812 0.0464>,<at 0.7034 0.0331>]}
         GreenCurve {keys [<at 0 0.5029>,<at 0.2812 0.35>,<at 0.7034 0.25>]}
         RedCurve {keys [<at 0 0>,<at 0.2812 0>,<at 0.7034 0>]}
        }}))
    )
    ScaleA=true
    ScaleRGB=true
   )
  ]
  rendererClasses [&NiagaraRibbonRendererProperties]
  renderers [{rendererClass=&NiagaraRibbonRendererProperties}]
 }
}
summary {
 systemName=NS_Pickaxe_CyborgWarrior_Trail
 emitters [
  {emitterName=MainTrail, rendererClasses [&NiagaraRibbonRendererProperties]}
  {emitterName=DarkTrail, rendererClasses [&NiagaraRibbonRendererProperties]}
  {emitterName=HandleTrail, rendererClasses [&NiagaraRibbonRendererProperties]}
 ]
 userVariables [User.RibbonIDOffset {type=NiagaraInt32}]
}
